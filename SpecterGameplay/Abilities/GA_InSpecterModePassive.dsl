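; Passive gameplay ability for Specter mode that swaps the player's inventory for Specter items while active.
; On activation it adds a looping local cue for the controlling client, then on authority clears the inventory and, per scalable-float toggles, grants Updoot and Hands items, force-equipping Updoot.
; Grants check for a BP_SpecterPawn and free inventory space.
; On end it removes those items and the cue, without restoring the cleared inventory.
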(event Ability|EventActivateAbility
  (|CallParentFunction)
  (Variables|Default|SetFortPawn (CallFunction|GetAvatarasFortPlayerPawn))
  (CallFunction|AddLocalGameplayCue)
  (Ability|HasAuthority
    (:True
      ;[1 1 1] Clear out the inventory (Pickaxe)
      (Default|ClearInventory)
      (Utilities|FlowControl|Sequence
        (:then_0
          (if (CallFunction|ShouldAutoGiveUpdootWeapon)
            (CallFunction|AddUpdootItem)))
        (:then_1
          (if (CallFunction|ShouldAutoGiveHands)
            (CallFunction|AddHandsItem)))))))

(event Ability|EventOnEndAbility
  (params (bWasCancelled :type Boolean))
  (|CallParentFunction bWasCancelled)
  (Utilities|FlowControl|Sequence
    (:then_0
      (if (CallFunction|ShouldAutoGiveUpdootWeapon)
        (Default|RemoveItem (Variables|Default|GetUpdoot%20Item%20Definition))))
    (:then_1
      (if (CallFunction|ShouldAutoGiveHands)
        (Default|RemoveItem (Variables|Default|GetHands%20Item%20Definition))))
    (:then_2
      (CallFunction|RemoveLocalGameplayCue))))

(event Custom|AddUpdootItem
  (bind _fortpawn (Variables|Default|GetFortPawn))
  ;[1 1 1] Grant and equip the Updoot item.
  (bind _asfort_player_controller (Utilities|Casting|CastToFortPlayerController (Pawn|GetController _fortpawn)))
  (Utilities|Casting|CastToBP_SpecterPawn _fortpawn
    (:then
      _)
    (:CastFailed
      (Fortnite|FortShippingLog _asfort_player_controller "ERROR: Failed to cast pawn to BP_SpecterPawn." true)
      (Utilities|FlowControl|Delay 5.0)))
  (Default|GrantItem (Variables|Default|GetUpdoot%20Item%20Definition) 1)
  (Inventory|ExecuteInventoryItemDefinition :self _asfort_player_controller :ItemDefinition (Variables|Default|GetUpdoot%20Item%20Definition) :Delay 0.0 :bForceExecute true :bActivateSlotAfterSettingFocused true))

(event Custom|AddHandsItem
  (bind _fortpawn (Variables|Default|GetFortPawn))
  ;[1 1 1] Grant the hands item.
  (bind _asfort_player_controller (Utilities|Casting|CastToFortPlayerController (Pawn|GetController _fortpawn)))
  (Utilities|Casting|CastToBP_SpecterPawn _fortpawn
    (:then
      _)
    (:CastFailed
      (Fortnite|FortShippingLog _asfort_player_controller "ERROR: Failed to cast pawn to BP_SpecterPawn when adding hands item." true)
      (Utilities|FlowControl|Delay 5.0)))
  (Default|GrantItem (Variables|Default|GetHands%20Item%20Definition) 1))

(event Custom|AddLocalGameplayCue
  ;[0.000475 0.000525 0.08854 0.5] Add/Remove cues on locally controlled pawn.
  (Ability|IsLocallyControlled
    (:True
      (bind _fortpawn (Variables|Default|GetFortPawn)
        (:then
          (Projectile|AddGameplayCue%28Looping%29Local _fortpawn (Variables|GameplayCues|GetGC_InSpecterModeLocal) (Abilities|MakeOwnerInstigatorCueParams self)))
        (:else)))))

(event Custom|RemoveLocalGameplayCue
  ;[0.000475 0.000525 0.08854 0.5] Add/Remove cues on locally controlled pawn.
  (Ability|IsLocallyControlled
    (:True
      (bind _fortpawn (Variables|Default|GetFortPawn)
        (:then
          (Projectile|RemoveGameplayCue%28Looping%29Local _fortpawn (Variables|GameplayCues|GetGC_InSpecterModeLocal) (Abilities|MakeOwnerInstigatorCueParams self)))
        (:else)))))

(fn ClearInventory
  ;[1 1 1] Clear out the inventory
  (bind _asfort_player_controller (Utilities|Casting|CastToFortPlayerController (Pawn|GetController (Variables|Default|GetFortPawn))))
  (bind _items (InventoryManagement|GetItemsfromInventoryOwner _asfort_player_controller))
  (bind (_arrayindex _arrayelement) (Utilities|Array|ReverseforEachLoop _items)
    (:LoopBody
      ;;[1 1 1] Note: removing negative amount means remove all.
      (Inventory|RemoveFortItemFromPlayer _asfort_player_controller _arrayelement -1 true))))

(fn GrantItem
  (params (ItemDefinition :type Fort%20World%20Item%20Definition%20Object%20Reference) (NumberToGive :type Integer))
  (bind _asfort_inventory_owner_interface (Utilities|Casting|CastToFortInventoryOwnerInterface (Ability|GetPlayerControllerfromActorInfo)))
  (if (>= (Inventory|GetInventorySpace _asfort_inventory_owner_interface) 1)
    (Inventory|GiveItemtoInventoryOwnerwithParams :InventoryOwner _asfort_inventory_owner_interface :ItemDefinition ItemDefinition :NumberToGive NumberToGive :Params (LocalVariables|GetLocal_GiveItemParams))))

(fn RemoveItem
  (params (ItemDefinition :type Fort%20World%20Item%20Definition%20Object%20Reference))
  (bind _asfort_inventory_owner_interface (Utilities|Casting|CastToFortInventoryOwnerInterface (Ability|GetPlayerControllerfromActorInfo)))
  (Inventory|RemoveItemfromInventoryOwner _asfort_inventory_owner_interface ItemDefinition))

(fn ShouldAutoGiveHands
  (return (ScalableFloat|GetValueasBool (Variables|Default|GetScalableFloat_ShouldAutoGiveHands))))

(fn ShouldAutoGiveUpdootWeapon
  (return (ScalableFloat|GetValueasBool (Variables|Default|GetScalableFloat_ShouldAutoGiveUpdootWeapon))))
