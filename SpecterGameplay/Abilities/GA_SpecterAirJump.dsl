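; Specter air jump gameplay ability, which runs only while the pawn has the active Specter movement tag and otherwise ends immediately.
; The non-server side applies guided-motion movement, commits, then fires a local burst cue and calls GhostJumped on the Specter pawn.
; The server only commits, and an OnlyServerWait net sync leads to ending the ability locally.
; Ending the ability cancels the Specter leap movement-mode extension.
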(event Ability|EventActivateAbility
  (bind _returnvalue (Ability|GetActivatingPawn))
  (Variables|Default|SetPawn _returnvalue)
  (if (GameplayTags|HasMatchingGameplayTag _returnvalue "(TagName=\"Granted.Specter.Movement.Active\")")
    (CallFunction|AirJump)
    (else
      (Ability|EndAbility))))

(event Custom|AirJump
  (if (not (Networking|IsServer))
    (bind _pawn (Variables|Default|GetPawn)
      (:then
        (Game|ApplyGuidedMotionMME _pawn (Variables|Default|GetAirJumpMME)
          (:then
            (bind _returnvalue (Ability|CommitAbility))
            (if _returnvalue
              (CallFunction|FireAirJumpCue)
              (CallFunction|SendAirJumpEventtoPawn)))
          (:On%20Finished
            _)))
      (:else))
    (else
      (Ability|CommitAbility)))
  (CallFunction|NetSyncandEnd))

(event Custom|NetSyncAndEnd
  (Ability|Tasks|WaitNetSync "OnlyServerWait"
    (:OnSync
      (Ability|EndAbilityLocally))))

(event Ability|EventOnEndAbility
  (params (bWasCancelled :type Boolean))
  (Class|FortMovementModeExtBlueprintLibrary|CancelActiveExtensionwithTag (Variables|Default|GetPawn) "(GameplayTags=((TagName=\"MovementMode.Leap.Specter\")))"))

(fn FireAirJumpCue
  (GameplayCue|ExecuteGameplayCue%28Burst%29Local (Variables|Default|GetPawn) (Variables|Default|GetAirJumpCue) (Ability|GameplayCue|MakeGameplayCueParameters :NormalizedMagnitude 0.0 :RawMagnitude 0.0 :EffectContext (Ability|GetContextfromOwner) :Location "0,0,0" :Normal "0,0,0" :Instigator (Variables|Default|GetPawn) :EffectCauser (Variables|Default|GetPawn) :GameplayEffectLevel 1 :AbilityLevel 1 :bReplicateLocationWhenUsingMinimalRepProxy false)))

(fn SendAirJumpEventToPawn
  (bind _asbp_specter_pawn (Utilities|Casting|CastToBP_SpecterPawn (Variables|Default|GetPawn)))
  (Class|BPSpecterPawn|GhostJumped _asbp_specter_pawn))
