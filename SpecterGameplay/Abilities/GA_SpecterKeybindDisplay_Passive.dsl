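; Passive Specter ability that keeps the on-screen keybind display in sync with flying and haunting states.
; On activation it starts gameplay tag add/remove listeners for both states, caching each as a boolean.
; Flight changes switch the "Jump" label between start and stop text unless haunting.
; Haunting shows the "Use" action and hides "Jump", whose visibility is also gated by the CanFly scalable float.
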(event Ability|EventActivateAbility
  (|CallParentFunction)
  (CallFunction|HandleFlightChanges)
  (CallFunction|HandleHauntChanges))

(event Ability|EventOnEndAbility
  (params (bWasCancelled :type Boolean)))

(event Custom|HandleFlightChanges
  (Ability|Tasks|WaitGameplayTagAdd (Variables|Default|GetFlyingTag)
    (:then
      (Ability|Tasks|WaitGameplayTagRemove (Variables|Default|GetFlyingTag)
        (:Removed
          (bind _output_get (Variables|Default|SetIsFlying false))
          (SpecterKeybindDisplay|UpdateFlightDisplayName _output_get))))
    (:Added
      (bind _output_get_2 (Variables|Default|SetIsFlying true))
      (SpecterKeybindDisplay|UpdateFlightDisplayName _output_get_2))))

(event Custom|HandleHauntChanges
  (bind _ishaunting (Variables|Default|GetIsHaunting))
  (Ability|Tasks|WaitGameplayTagAdd (Variables|Default|GetHauntingTag)
    (:then
      (Ability|Tasks|WaitGameplayTagRemove (Variables|Default|GetHauntingTag)
        (:Removed
          (Variables|Default|SetIsHaunting false)
          (SpecterKeybindDisplay|UpdateFlightDisplayName (Variables|Default|GetIsFlying))
          (SpecterKeybindDisplay|SetFlightActionVisibility (not _ishaunting))
          (SpecterKeybindDisplay|SetHauntActionVisivility _ishaunting))))
    (:Added
      (bind _output_get (Variables|Default|SetIsHaunting true))
      (SpecterKeybindDisplay|SetFlightActionVisibility (not _output_get))
      (SpecterKeybindDisplay|SetHauntActionVisivility (Variables|Default|GetIsHaunting)))))

(fn SetFlightActionVisibility
  (params (Visible :type Boolean))
  (CallFunction|SetActionDisplayVisibility "Jump" (and Visible (ScalableFloat|GetValueasBool (Variables|Default|GetCanFly)))))

(fn SetHauntActionVisivility
  (params (Visible :type Boolean))
  (CallFunction|SetActionDisplayVisibility "Use" Visible))

(fn UpdateFlightDisplayName
  (params (IsFlying :type Boolean))
  (if (not (Variables|Default|GetIsHaunting))
    (CallFunction|SetActionDisplayText "Jump" (Utilities|String|SelectText (Variables|Default|GetStopFlyingText) (Variables|Default|GetStartFlyingText) (Variables|Default|GetIsFlying)))))
