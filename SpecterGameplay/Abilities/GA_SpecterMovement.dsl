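; Gameplay ability that gives the Specter pawn its ghost movement: tuned air control and gravity, plus double-tap-jump flight.
; Activation caches and overrides movement values, adds a local looping cue and binds a jump cue; ending reverses these.
; A timed double tap, gated by cooldown, CanFly and no Haunting tag, toggles a flight movement mode extension, which gaining that tag cancels.
; Gravity restores to a hardcoded 1.0, as the real value is not Blueprint-accessible.
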(event Ability|EventActivateAbility
  ;[0 0.2783 0 0.5] ActivateAbility
  (|CallParentFunction)
  (bind _asspecter_pawn (Utilities|Casting|CastToSpecterPawn (Ability|GetActivatingPawn))
    (:then
      (CharacterMovement|UpdateCharacterMovementValues _asspecter_pawn)
      (Pawn|BindEventtoOnPawnJumped _asspecter_pawn (EventDispatchers|CreateEvent :!event Custom|OnPawnJumped))
      (CallFunction|AddLocalGameplayCue)
      (CallFunction|ListenforHauntingChanges))
    (:CastFailed
      (Ability|EndAbility))))

(event Ability|EventOnAbilityInputPressed
  ;[0 0 0] Double-tap Jump quickly to toggle flight
  (bind (_cantoggleflight _serverworldtimeseconds) (CharacterMovement|Flight|CanToggleFlight))
  (if _cantoggleflight
    (Variables|CharacterMovement|Flight|SetLastJumpInputTime _serverworldtimeseconds)
    (CallFunction|ToggleFlight)
    (else
      (Variables|CharacterMovement|Flight|SetLastJumpInputTime _serverworldtimeseconds))))

(event Custom|ToggleFlight
  ;[0 0 0 0.5] Toggle Flight On/Off
  (Variables|CharacterMovement|Flight|SetLastFlightToggleTime (Variables|CharacterMovement|Flight|GetLastJumpInputTime))
  (bind _returnvalue (Ability|GetActivatingPawn))
  (CharacterMovement|Flight|SetFlightMode (not (Ability|DoesTargetHaveTags _returnvalue (Variables|CharacterMovement|Flight|GetMovmentGrantedFlightTags))) _returnvalue))

(event Ability|EventOnEndAbility
  (params (bWasCancelled :type Boolean))
  ;[0.1406 0.02345 0] End Ability
  (CallFunction|StopListenforHauntingChanges)
  (CallFunction|RemoveLocalGameplayCue)
  (bind _returnvalue (Ability|GetActivatingPawn))
  (Utilities|IsValid _returnvalue
    (:Is%20Valid
      (CharacterMovement|ResetCharacterMovementValues _returnvalue)
      (CharacterMovement|Flight|SetFlightMode false _returnvalue)
      (Pawn|UnbindEventfromOnPawnJumped _returnvalue (EventDispatchers|CreateEvent :!event Custom|OnPawnJumped)))))

(event Custom|ListenForHauntingChanges
  ;[0 0 0 0.5] Cancel flight when the ghost haunts a player.
  (bind _asynctaskproxy (Ability|Tasks|WaitGameplayTagAdd (Variables|CharacterMovement|Flight|GetHauntingTag))
    (:then
      (Variables|CharacterMovement|Flight|SetHauntingTagTask _asynctaskproxy))
    (:Added
      (bind _returnvalue (Ability|GetActivatingPawn))
      (if (Ability|DoesTargetHaveTags _returnvalue (Variables|CharacterMovement|Flight|GetMovmentGrantedFlightTags))
        (CharacterMovement|Flight|SetFlightMode false _returnvalue)))))

(event Custom|StopListenForHauntingChanges
  ;[0 0 0 0.5] Cancel flight when the ghost haunts a player.
  (bind _hauntingtagtask (Variables|CharacterMovement|Flight|GetHauntingTagTask)
    (:then
      (GameplayTasks|EndTask _hauntingtagtask))
    (:else)))

(event Custom|AddLocalGameplayCue
  ;[0.000475 0.000525 0.08854 0.5] Add/Remove cues on locally controlled pawn.
  (Ability|IsLocallyControlled
    (:True
      (Projectile|AddGameplayCue%28Looping%29Local (Ability|GetActivatingPawn) (Variables|Default|GetGCSpecterMovementLocal) (Abilities|MakeOwnerInstigatorCueParams self)))))

(event Custom|RemoveLocalGameplayCue
  ;[0.000475 0.000525 0.08854 0.5] Add/Remove cues on locally controlled pawn.
  (Ability|IsLocallyControlled
    (:True
      (Projectile|RemoveGameplayCue%28Looping%29Local (Ability|GetActivatingPawn) (Variables|Default|GetGCSpecterMovementLocal) (Abilities|MakeOwnerInstigatorCueParams self)))))

(event Custom|OnPawnJumped
  ;[0.000475 0.000526 0.08854 0.5] Fire Cue on Jump from the ground
  (Ability|ExecuteGameplayCueOnOwner (Variables|CharacterMovement|GetJumpCueTag) (Ability|GetContextfromOwner)))

(fn CanToggleFlight
  (if (and (ScalableFloat|GetValueasBool (Variables|CharacterMovement|Flight|GetCanFly)) (not (CallFunction|IsHaunting)))
    ;[0 0 0 0.5] Check if the player pressed jump fast enough AND they've waited long enough since the last time they toggled flight.
    (bind _returnvalue (Fortnite|Sync|Time|GetServerWorldTimeSeconds))
    (return (and (< (- _returnvalue (Variables|CharacterMovement|Flight|GetLastJumpInputTime)) (ScalableFloat|GetValueatLevel (Variables|CharacterMovement|Flight|GetDoubleTapForFlightJumpTime))) (> (- _returnvalue (Variables|CharacterMovement|Flight|GetLastFlightToggleTime)) (ScalableFloat|GetValueatLevel (Variables|CharacterMovement|Flight|GetFlightToggleCooldownTime)))) _returnvalue)
    (else
      (return false 0.0))))

(fn IsHaunting
  (return (GameplayTags|HasMatchingGameplayTag (Ability|GetAbilitySystemComponent (Ability|GetActivatingPawn)) (Variables|CharacterMovement|Flight|GetHauntingTag))))

(fn ResetCharacterMovementValues
  (params (Pawn :type Fort%20Pawn%20Object%20Reference))
  (if (Variables|CharacterMovementCache|GetDidCacheCharMovement)
    (CharacterMovement|SetCharacterMovement Pawn (Variables|CharacterMovementCache|GetCachedAirControl) (Variables|CharacterMovementCache|GetCachedAirControlBoostMult) (Variables|CharacterMovementCache|GetCachedAirControlBoostVelThresh) (Variables|CharacterMovementCache|GetCachedFallingLateralFriction) (Variables|CharacterMovementCache|GetCachedGravityMultiplier))
    (Variables|CharacterMovementCache|SetDidCacheCharMovement false)))

(fn SetCharacterMovement
  (params (Pawn :type Fort%20Pawn%20Object%20Reference) (Air%20Control :type Float%20%28single-precision%29) (Air%20Control%20Boost%20Multiplier :type Float%20%28single-precision%29) (Air%20Control%20Boost%20Velocity%20Threshold :type Float%20%28single-precision%29) (FallingLateralFriction :type Float%20%28double-precision%29) (GravityZMult :type Float%20%28double-precision%29))
  (bind _charactermovement (Variables|/Script/Engine.Character|GetCharacterMovement Pawn))
  (Variables|/Script/Engine.CharacterMovementComponent|SetAirControl Air%20Control _charactermovement)
  (Variables|/Script/Engine.CharacterMovementComponent|SetAirControlBoostMultiplier Air%20Control%20Boost%20Multiplier _charactermovement)
  (Variables|/Script/Engine.CharacterMovementComponent|SetAirControlBoostVelocityThreshold Air%20Control%20Boost%20Velocity%20Threshold _charactermovement)
  (Variables|/Script/Engine.CharacterMovementComponent|SetFallingLateralFriction FallingLateralFriction _charactermovement)
  (FortPawn|SetGravityMultiplier (Variables|GetPawn) GravityZMult))

(fn SetFlightMode
  (params (Enabled :type Boolean) (Pawn :type Fort%20Pawn%20Object%20Reference))
  (if Enabled
    (Class|FortMovementModeExtBlueprintLibrary|ActivateRegisteredMovementModeExtension (Variables|CharacterMovement|Flight|GetFlightMMEClass) Pawn)
    (else
      (Class|FortMovementModeExtBlueprintLibrary|CancelActiveExtensionwithTag Pawn "(GameplayTags=((TagName=\"MovementMode.Flying.Specter\")))"))))

(fn UpdateCharacterMovementValues
  (params (Pawn :type Fort%20Pawn%20Object%20Reference))
  (bind _charactermovement (Variables|/Script/Engine.Character|GetCharacterMovement Pawn))
  (Utilities|FlowControl|Sequence
    (:then_0
      (if (not (Variables|CharacterMovementCache|GetDidCacheCharMovement))
        (Variables|CharacterMovementCache|SetCachedAirControl (Variables|/Script/Engine.CharacterMovementComponent|GetAirControl _charactermovement))
        (Variables|CharacterMovementCache|SetCachedAirControlBoostMult (Variables|/Script/Engine.CharacterMovementComponent|GetAirControlBoostMultiplier _charactermovement))
        (Variables|CharacterMovementCache|SetCachedAirControlBoostVelThresh (Variables|/Script/Engine.CharacterMovementComponent|GetAirControlBoostVelocityThreshold _charactermovement))
        (Variables|CharacterMovementCache|SetCachedFallingLateralFriction (Variables|/Script/Engine.CharacterMovementComponent|GetFallingLateralFriction _charactermovement))
        (Variables|CharacterMovementCache|SetCachedGravityMultiplier 1.0) ; Pawn's Grav Mult in the FortAttributeSet \nis currently not BP Accessible. Hoping this \nis acceptable until it isn't and we can provide \nan accessor to cache it properly.
        (Variables|CharacterMovementCache|SetDidCacheCharMovement true)))
    (:then_1
      (CharacterMovement|SetCharacterMovement Pawn (ScalableFloat|GetValueatLevel (Variables|CharacterMovement|GetAirControl)) (ScalableFloat|GetValueatLevel (Variables|CharacterMovement|GetAirControlBoostMultiplier)) (ScalableFloat|GetValueatLevel (Variables|CharacterMovement|GetAirControlBoosVelocityThreshold)) (ScalableFloat|GetValueatLevel (Variables|CharacterMovement|GetFallingLateralFriction)) (ScalableFloat|GetValueatLevel (Variables|CharacterMovement|GetGravityMultiplier))))))
