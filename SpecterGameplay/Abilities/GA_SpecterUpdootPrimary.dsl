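; Gameplay ability for the Specter's primary Updoot action, which plays a montage and gives quest credit for dooting.
; Activation stores the activating pawn and commits the ability, ending it if either step fails.
; On commit it starts the firing montage, locks move input and rotate-to-camera, and broadcasts the CustomVerb.Specter.Updoot message.
; The ability ends when the montage completes or is cancelled, and OnEndAbility restores both input settings.
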(event Ability|EventActivateAbility
  (bind _returnvalue (Ability|GetActivatingPawn))
  ;[1 1 1] Activate Ability
  (|CallParentFunction)
  (Utilities|IsValid _returnvalue
    (:Is%20Valid
      (Variables|ReferencedActors|SetPlayerPawn _returnvalue)
      (bind _returnvalue_2 (Ability|CommitAbility))
      (if (not _returnvalue_2)
        (Ability|EndAbility)))
    (:Is%20Not%20Valid
      (Ability|EndAbility))))

(event Ability|EventCommitExecute
  ;[1 1 1] Commit Execute
  (|CallParentFunction)
  (Ability|Tasks|PlayMontageandWaitforTrigger "None" (Variables|Default|GetFiring%20Montage)
    (:then
      (bind _asfort_player_controller (Utilities|Casting|CastToFortPlayerController (Pawn|GetController (Variables|ReferencedActors|GetPlayerPawn))))
      (Input|SetIgnoreMoveInput _asfort_player_controller true)
      (Input|SetRotatePawntoCamera _asfort_player_controller false)
      (CallFunction|BroadcastDootCustomVerb))
    (:Cancelled
      _)
    (:Completed
      _))
  (Ability|EndAbility))

(event Custom|BroadcastDootCustomVerb
  ; Give quest credit for dooting
  (CustomVerb|BroadcastCustomVerbMessage (CustomVerb|MakeCustomVerbMessage (Variables|ReferencedActors|GetPlayerPawn) "(TagName=\"CustomVerb.Specter.Updoot\")")))

(event Ability|EventOnEndAbility
  (params (bWasCancelled :type Boolean))
  ;[1 1 1] On End Ability
  (|CallParentFunction bWasCancelled)
  (bind _asfort_player_controller (Utilities|Casting|CastToFortPlayerController (Pawn|GetController (Variables|ReferencedActors|GetPlayerPawn))))
  (Input|SetIgnoreMoveInput _asfort_player_controller false)
  (Input|SetRotatePawntoCamera _asfort_player_controller true))
