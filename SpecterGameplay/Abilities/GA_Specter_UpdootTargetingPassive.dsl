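; Passive Specter ability that keeps the Updoot weapon's lock-on target updated, running only without authority (locally on the autonomous client).
; StartTargeting prefers the haunted pawn; otherwise AllowBots selects the cone TargetingPreset.
; As wired, only the bot-allowed branch runs async targeting, picking hits ahead of the player by best camera-direction dot product above a minimum.
; The preferred target goes to the weapon's LockonTargetActor, and a short Delay re-calls StartTargeting to form a polling loop.
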(event Ability|EventActivateAbility
  (|CallParentFunction)
  (Variables|Default|SetPlayerPawn (CallFunction|GetAvatarasFortPlayerPawn))
  ; Update Targeting Locally (on Autonomous).
  (Ability|HasAuthority
    (:False
      (CallFunction|StartTargeting))))

(event Ability|EventOnEndAbility
  (params (bWasCancelled :type Boolean))
  (|CallParentFunction bWasCancelled))

(event Custom|StartTargeting
  (bind _returnvalue (Actor|GetComponentbyClass (Variables|Default|GetPlayerPawn) "/SpecterGameplay/Gameplay/BP_SpecterHauntingComponent.BP_SpecterHauntingComponent_C"))
  (Utilities|IsValid _returnvalue
    (:Is%20Valid
      (if (Class|SpecterHauntingComponent|IsHaunting _returnvalue)
        (Variables|Default|SetPreferredTargetActor (Class|SpecterHauntingComponent|GetHauntedPawn _returnvalue))
        (elif (ScalableFloat|GetValueasBool (Variables|Default|GetAllowBots))
          (Variables|Default|SetTargetingPreset "/SpecterGameplay/items/TP_SpecterUpdootConeAllowBots.TP_SpecterUpdootConeAllowBots")
          (bind (_asynctaskproxy _targetinghandle) (Class|AsyncActionPerformTargeting|PerformTargetingAsyncAction (Variables|Default|GetPlayerPawn) (Variables|Default|GetTargetingPreset) true)
            (:Targeted
              (bind _asplayer_controller (Utilities|Casting|CastToPlayerController (Pawn|GetController (Variables|Default|GetPlayerPawn))))
              (Variables|Default|SetCameraLocation (Camera|GetCameraLocation (Variables|/Script/Engine.PlayerController|GetPlayerCameraManager _asplayer_controller)))
              (Variables|Default|SetCameraDirNoramlized (Math|Vector|GetForwardVector (Camera|GetCameraRotation (Variables|/Script/Engine.PlayerController|GetPlayerCameraManager _asplayer_controller))))
              (Variables|Default|SetPreferredDotResult (Variables|Default|GetMin%20Dot%20Result%20To%20Target%20Player))
              (Variables|Default|SetPreferredTargetActor 0)
              (bind _asbp_specter_pawn (Utilities|Casting|CastToBP_SpecterPawn (Variables|Default|GetPlayerPawn)))
              (Variables|Default|SetIsPlayerHaunting (Variables|/Script/SpecterGameplayRuntime.SpecterPawn|GetbIsHaunting _asbp_specter_pawn))
              ; Loop through possible targets for the best one.
              (for _array_element (TargetingSystem|TargetingResults|GetTargetingResults (GameInstanceSubsystems|GetTargetingSubsystem) _targetinghandle)
                (bind _returnvalue_2 (Transformation|GetActorLocation (.HitActor (Collision|BreakHitResult _array_element))))
                (if (or (Variables|Default|GetIsPlayerHaunting) (>= (Math|Vector|DotProduct (- _returnvalue_2 (Transformation|GetActorLocation (Variables|Default|GetPlayerPawn))) (Variables|Default|GetCameraDirNoramlized)) 0))
                  (bind _returnvalue_3 (Math|Vector|DotProduct (Math|Vector|Normalize (- _returnvalue_2 (Variables|Default|GetCameraLocation))) (Variables|Default|GetCameraDirNoramlized)))
                  (if (or (Variables|Default|GetIsPlayerHaunting) (> _returnvalue_3 (Variables|Default|GetPreferredDotResult)))
                    (Variables|Default|SetPreferredDotResult _returnvalue_3)
                    (Variables|Default|SetPreferredTargetActor (.HitActor (Collision|BreakHitResult _array_element))))))))
          (else
            (Variables|Default|SetTargetingPreset "/SpecterGameplay/items/TP_SpecterUpdootCone.TP_SpecterUpdootCone"))))
      (Utilities|FlowControl|Sequence
        (:then_0
          ; When done, update Our weapon's Local Target Actor with our result.
          (bind _asb_specter_updoot_weapon (Utilities|Casting|CastToB_Specter_Updoot_Weapon (Weapon|GetCurrentWeapon (Variables|Default|GetPlayerPawn))))
          (Class|BSpecterUpdootWeapon|SetLockonTargetActor _asb_specter_updoot_weapon (Variables|Default|GetPreferredTargetActor)))
        (:then_1
          ; Wait then start custom targeting again.
          (Utilities|FlowControl|Delay 0.2)
          (CallFunction|StartTargeting))))))
